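; Zero Point orb FX actor that drives the orb's Niagara effect and feeds global vortex materials.
; On non-authority BeginPlay, it sets the Fortnite material parameter collection's vortex strength and centers it on the orb.
; DoPop fires the Niagara "Pop" parameter, then re-arms a timer for the soonest of seven staggered loops, syncing with shard animation.
; CalculateLoopTimes derives those periods and offsets using 8-bit-quantized interpolation, though no call to it appears here.
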(event EventBeginPlay
  (Networking|SwitchHasAuthority
    (:Remote
      (Rendering|Material|SetScalarParameterValue "/Game/Packages/Fortress_SharedMaterials/GlobalMaterialParameters/FortniteMaterialParameters.FortniteMaterialParameters" "AthenaVortexStrength" 1.0 :!class /Script/Engine.KismetMaterialLibrary)
      (Rendering|Material|SetVectorParameterValue "/Game/Packages/Fortress_SharedMaterials/GlobalMaterialParameters/FortniteMaterialParameters.FortniteMaterialParameters" "AthenaVortexCenter" (Math|Conversions|ToLinearColor(Vector) (Transformation|GetWorldLocation (Variables|Default|GetNS_Env_ZeroPoint_Orb))) :!class /Script/Engine.KismetMaterialLibrary))))

(event Collision|EventActorBeginOverlap
  (params (OtherActor :type Actor%20Object%20Reference)))

(event EventTick
  (params (DeltaSeconds :type Float%20%28single-precision%29)))

(event Custom|EventDoPop
  ; Sets a parameter on the particle system, and then sets an new timer to sync up with the Zero Point orb shard animation
  (Default|DoPop)
  (bind _delay (Default|GetNextPopDelay))
  (Utilities|Time|SetTimerbyEvent (delegate Custom|EventDoPop) _delay))

(fn CalculateLoopTimes
  (bind _returnvalue (/ 0.5 8.0))
  (bind _poploopmin (Variables|Default|GetPopLoopMin))
  (bind _poploopmax (Variables|Default|GetPopLoopMax))
  (Utilities|Array|Resize (Variables|Default|GetPopLoopTimes) 7)
  (Utilities|Array|Resize (Variables|Default|GetPopLoopOffsets) 7)
  (for _index (range 7)
    (Utilities|Array|SetArrayElem (Variables|Default|GetPopLoopTimes) _index (Math|Float|Lerp _poploopmin _poploopmax (/ (Math|Conversions|ToFloat(Integer) (Math|Float|Round (* (+ (/ (Math|Conversions|ToFloat(Integer) (+ _index 1)) 8.0) _returnvalue) 255.0))) 255.0)))
    (Utilities|Array|SetArrayElem (Variables|Default|GetPopLoopOffsets) _index (+ (* (Math|Float|Lerp _poploopmin _poploopmax (/ (Math|Conversions|ToFloat(Integer) (Math|Float|Round (* (+ (/ (Math|Conversions|ToFloat(Integer) (+ _index 1)) 8.0) _returnvalue) 255.0))) 255.0)) (/ (Math|Conversions|ToFloat(Integer) (Math|Float|Round (* (+ (/ (Math|Conversions|ToFloat(Integer) (+ _index 1)) 8.0) _returnvalue) 255.0))) 255.0)) -1.0))))

(fn DoPop
  (Niagara|SetNiagaraVariable(Float) (Variables|Default|GetNS_Env_ZeroPoint_Orb) "Pop" 1.0))

(fn GetNextPopDelay
  (LocalVariables|SetSoonestPopDelay 1e+08)
  (for (_array_element _array_index) (Variables|Default|GetPopLoopTimes)
    (LocalVariables|SetSoonestPopDelay (Math|Float|Min(Float) (- _array_element (Math|Float|%25%28Float%29 (+ (Utilities|Time|GetTimeSeconds) (Utilities|Array|Get(aref) (Variables|Default|GetPopLoopOffsets) _array_index)) _array_element)) (LocalVariables|GetSoonestPopDelay))))
  (return (LocalVariables|GetSoonestPopDelay)))

(fn ConstructionScript)
