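# Unlit translucent material for a flickering center glow on the Zero Point orb, drawn as a camera-facing sprite.
# Layered sines, varied by actor position or particle random, flicker brightness, opacity and corona size.
# Emissive combines a sphere-mask core, a gradient-remapped corona and a flare mask, with optional time-of-day blend.
# WPO handles sprite scaling, near-camera shrink and camera offset, while opacity adds depth fade and optional camera fade or daytime shutoff.
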Material /Game/Environments/Hera/FX/ZeroPoint/Orb/Materials/M_Env_ZeroPoint_Orb_CenterGlowFlicker
graph {
 properties {
  bAutomaticallySetUsageInEditor=false
  blendMode=BLEND_Translucent
  previewMesh=/Game/Effects/Fort_Effects/Effects/Environment/Hera/Shards/Glow/SM_Env_TeaCake_Shard_Glows
  shadingModel=MSM_Unlit
  twoSided=true
  usedShadingModels=Unlit
 }
 nodes [
  MaterialFunctionCall(Emissive=NamedRerouteUsage()) {materialFunction=&MF_FX_EmissiveAdjustments}
  n1 = Multiply(A=Time(), B=ScalarParameter() {defaultValue=1, parameterName=Flicker Speed Multiplier})
  n2 = MaterialFunctionCall(Float3=ActorPositionWS()) {materialFunction=&BreakOutFloat3Components}
  n3 = Add(A=Add(A=@n2.R, B=@n2.R), B=@n2.B)
  n4 = LinearInterpolate(A=Multiply(A=@n1) {constB=0.7}, B=Multiply(A=@n1) {constB=1.5}, Alpha=Frac(None=Divide(A=@n3) {constB=50}))
  n5 = Sine(None=@n4) {period=0.5651}
  n6 = Sine(None=@n4) {period=0.442}
  n7 = Add(A=Multiply(A=Add(A=@n5, B=@n6), B=ScalarParameter() {defaultValue=0.15, group=Emissive, parameterName=GlowFlickerScale}))
  n8 = Constant() {r=0.5}
  n9 = SphereMask(A=TextureCoordinate(), B=@n8, Radius=@n8) {hardnessPercent=0}
  n10 = Constant() {r=0.5}
  n11 = ParticleRandom()
  n12 = Multiply(
   A=Time()
   B=StaticSwitchParameter(True=LinearInterpolate(Alpha=@n11) {constA=1.2, constB=0.8}, False=Constant() {r=1}) {parameterName=GPU Particles}
  )
  n13 = Sine(None=@n12) {period=0.075}
  n14 = Multiply(A=Multiply(A=Sine(None=@n12) {period=0.155}, B=Sine(None=@n12) {period=0.2}), B=Sine(None=@n12) {period=0.43})
  n15 = ScalarParameter() {defaultValue=0.4, desc="0 = full flicker amount", parameterName=GPUFlickerFadeMin}
  n16 = ScalarParameter() {defaultValue=1, desc="1 = no flicker", parameterName=GPUFlickerFadeMax, sortPriority=1}
  n17 = ParticleRandom()
  n18 = ScalarParameter() {defaultValue=1, parameterName=FlickerLerp}
  n19 = DynamicParameter() {paramNames [FlickerFade,CoronaSize,Param3,Size]}
  n20 = LinearInterpolate(
   A=Clamp(None=Multiply(A=@n13, B=@n14)) {maxDefault=3}
   Alpha=StaticSwitchParameter(True=LinearInterpolate(A=@n15, B=@n16, Alpha=@n17) {constA=0.5}, False=Multiply(A=@n18, B=@n19.FlickerFade)) {parameterName=GPU Particles}
  ) {constA=1}
  n21 = ScalarParameter() {defaultValue=1, parameterName=CoronaSize}
  n22 = Add(
   A=LinearInterpolate(Alpha=@n20) {constA=-0.4, constB=0}
   B=StaticSwitchParameter(True=@n21, False=Multiply(A=@n19.CoronaSize, B=@n21)) {parameterName=GPU Particles}
  ) {constB=0}
  n23 = TextureSample(UVs=SphereMask(A=TextureCoordinate(), B=@n10, Radius=@n10, Hardness=LinearInterpolate(Alpha=@n22) {constA=0.7, constB=-0.3}) {hardnessPercent=25}) {texture=/Game/Effects/Fort_Effects/Textures/Grad/T_Halo_Corona_Gradient}
  n24 = Constant() {r=1}
  n25 = Multiply(
   A=StaticSwitchParameter(True=Desaturation(None=@n23.RGB, Fraction=@n24), False=@n23.RGB) {parameterName=UseDesaturatedFlare?}
   B=ScalarParameter() {defaultValue=0.125, parameterName=CoronaBrightness}
  ) {constB=0.125}
  n26 = Add(A=Multiply(A=Power(Base=@n9) {constExponent=25}, B=ScalarParameter() {defaultValue=2, parameterName=CenterBrightness}) {constB=2}, B=@n25)
  n27 = VectorParameter() {parameterName=Color, defaultValue=<rgba 1 1 1 1>}
  n28 = ParticleColor()
  n29 = StaticSwitchParameter(True=@n27.RGB, False=Multiply(A=@n28.RGB, B=@n27.RGB)) {parameterName=UsedWithSkeletalMesh}
  n30 = Multiply(A=Power(Base=@n9) {constExponent=4}, B=ScalarParameter() {defaultValue=0.15, parameterName=EdgeBrightness}) {constB=0.15}
  n31 = TextureSampleParameter2D() {parameterName=FlareMask, texture=/Game/Effects/Fort_Effects/Textures/Flares/T_Flare_Spikeball}
  n32 = Add(
   A=Multiply(A=ShadingPathSwitch(Default=@n26, Forward=@n25), B=LinearInterpolate(B=@n29) {constA=1, constAlpha=0.7})
   B=Multiply(A=Add(A=@n30, B=@n31.RGB), B=@n29)
  )
  n33 = ScalarParameter() {defaultValue=1, parameterName="Value_Morning (S)"}
  n34 = ScalarParameter() {defaultValue=1, parameterName="Value_Day (S)"}
  n35 = ScalarParameter() {defaultValue=1, parameterName="Value_Evening (S)"}
  n36 = ScalarParameter() {defaultValue=1, parameterName="Value_Night (S)"}
  NamedRerouteDeclaration(None=LinearInterpolate(
    A=Multiply(A=@n7, B=@n32)
    B=MaterialFunctionCall(Value_Morning=@n33, Value_Day=@n34, Value_Evening=@n35, Value_Night=@n36) {materialFunction=&Scalar_Time_of_Day}.Result
    Alpha=ScalarParameter() {parameterName=UseTimeOfDay?}
   )) {name=Color, nodeColor=<rgba 0.5469 0 1 1>}
  n37 = WorldPosition()
  n38 = ScalarParameter() {defaultValue=30, parameterName=SpriteScale}
  n39 = MaterialFunctionCall(Distance=ScalarParameter() {defaultValue=5000, group=WPO, parameterName=SpriteCamGrowDistance}) {materialFunction=&Near_Camera_Fade}
  n40 = StaticSwitchParameter(
   True=LinearInterpolate(A=@n38, B=ScalarParameter() {defaultValue=30, parameterName=SpriteScaleMax}, Alpha=Power(Base=@n39.Result) {constExponent=4})
   False=@n38
  ) {parameterName=UseMaxDistanceScaling?}
  n41 = Power(Base=MaterialFunctionCall(Distance=ScalarParameter() {defaultValue=150, parameterName=NearCameraShrink}) {materialFunction=&Near_Camera_Fade}.Result) {constExponent=3}
  n42 = Multiply(
   A=Multiply(A=@n19.Size, B=Multiply(A=@n40, B=@n41))
   B=AppendVector(A=Constant() {r=1}, B=ScalarParameter() {defaultValue=1, parameterName=SpriteScaleVertical})
  )
  n43 = VectorParameter()
  n44 = ScalarParameter() {parameterName=CameraOffset}
  NamedRerouteDeclaration(None=Add(
    A=MaterialFunctionCall!(Center Location=@n37.XYZ, 0-1 UVs, XY scale=@n42, OverridePivot, "Normals (Optional)"=@n43.RGB) {materialFunction=&Sprite}.World position offset
    B=MaterialFunctionCall(Offset Amount=@n44) {materialFunction=/Engine/Functions/Engine_MaterialFunctions02/WorldPositionOffset/CameraOffset}.Camera Offset
   )) {name=WPO, nodeColor=<rgba 1 0 0.0938 1>}
  n45 = MaterialFunctionCall(Distance=ScalarParameter() {defaultValue=10, parameterName=NearCamFadeDistance}) {materialFunction=&Near_Camera_Fade}
  n46 = Constant() {r=1}
  n47 = TextureSampleParameter2D() {parameterName=OuterFlareMaskTexture, samplerType=SAMPLERTYPE_Grayscale, texture=/Game/Effects/Fort_Effects/Textures/T_EFX_CameraFlare_01}
  n48 = StaticSwitchParameter(True=@n27.A, False=Multiply(A=@n28.A, B=@n27.A)) {parameterName=UsedWithSkeletalMesh}
  n49 = DepthFade(Opacity=@n48, FadeDistance=ScalarParameter() {defaultValue=20, parameterName=DepthFade})
  n50 = Multiply(A=@n20, B=Multiply(A=ShadingPathSwitch(Default=@n49, Deferred=@n49, Mobile=@n48) {desc=Migrated from FeatureLevelSwitch}, B=@n41))
  n51 = Power(
   Base=StaticSwitchParameter(True=Multiply(A=@n47.G, B=@n50), False=@n50) {parameterName=UseFlareMask}
   Exp=ScalarParameter() {defaultValue=0.4, parameterName=OpacityPow}
  )
  n52 = CollectionParameter() {
   collection=&FortniteMaterialParameters
   parameterName=FortniteActiveTimeOfDay
  }
  n53 = LinearInterpolate(B=@n51, Alpha=DotProduct(A=@n52, B=Constant4Vector() {constant=<rgba 0 0 0 1>}.RGBA))
  NamedRerouteDeclaration(None=Multiply(
    A=StaticSwitchParameter(True=@n45.Result, False=@n46) {parameterName=UseNearCamFade?}
    B=StaticSwitchParameter(True=@n53, False=@n51) {parameterName=Turn Off At Day}
   )) {name=Opacity, nodeColor=<rgba 1 0.5206 0.2152 1>}
  NamedRerouteUsage()
  NamedRerouteUsage()
 ]
}
