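# Unlit translucent material for the dark swirl of the Zero Point orb effect, drawn as a sprite that shrinks to zero over a distance fade.
# Noise-distorted, panning radial UVs, sampled with derivative mips to handle the polar seam, build swirling color and opacity masks.
# Color blends Fresnel-driven tints with a faint screen-aligned nebula and darkens toward the center.
# Opacity fades with distance, with depth fade only on higher-end platforms.
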Material /Game/Environments/Hera/FX/ZeroPoint/Orb/Materials/M_Env_ZeroPoint_Orb_DarkSwirl
graph {
 properties {bAutomaticallySetUsageInEditor=false, blendMode=BLEND_Translucent, shadingModel=MSM_Unlit, usedShadingModels=Unlit}
 nodes [
  Constant() {r=0.92}
  n1 = ScalarParameter() {defaultValue=65000, group=Scale, parameterName=SpriteScale_DistanceFade_EndFading}
  n2 = ScalarParameter() {defaultValue=45000, group=Scale, parameterName=SpriteScale_DistanceFade_StartFading}
  MaterialFunctionCall!(
   Center Location
   0-1 UVs
   XY scale=LinearInterpolate(
    A=ScalarParameter() {defaultValue=3100}
    Alpha=MaterialFunctionCall(Fade_Start_Distance=@n1, Fade_End_Distance=@n2) {materialFunction=&MF_Distance_SyncedFade_FX}.Result
   ) {constB=0}
   OverridePivot
   "Normals (Optional)"
  ) {materialFunction=&Sprite}
  n3 = Constant() {r=100}
  n4 = TextureSample(UVs=Panner(Coordinate=TextureCoordinate()) {speedY=-0.2}) {samplerType=SAMPLERTYPE_LinearGrayscale, texture=/Game/Effects/Fort_Effects/Emotes/UltraLight/Texture/T_DirectionalNoise_Greyscale}
  n5 = TextureSample(UVs=Add(
    A=MaterialFunctionCall!(TextureResolution=@n3) {materialFunction=/Engine/Functions/Engine_MaterialFunctions02/Texturing/ScreenAlignedPixelToPixelUVs}.Result
    B=Multiply(A=@n4.R) {constB=0.3}
   )) {texture=/Game/Characters/Player/Male/Medium/Bodies/M_Med_Soldier_04/Skins/Galaxy/Textures/T_FN_Nebula_style2}
  n6 = TextureCoordinate() {coordinateIndex=1}
  n7 = Constant() {r=0.5}
  n8 = Constant() {r=0.4}
  n9 = Multiply(
   A=Desaturation(None=Multiply(A=@n5.RGB) {constB=10}, Fraction=Constant() {r=-1})
   B=OneMinus(None=SphereMask(A=@n6, B=@n7, Radius=@n8) {hardnessPercent=50})
  )
  n10 = VectorParameter() {parameterName=Param_1, defaultValue=<rgba 0.3333 0 0.0903 1>}
  n11 = VectorParameter() {defaultValue=<rgba 0.0505 0.0494 0.0485 1>}
  n12 = Fresnel() {baseReflectFraction=0, exponent=1}
  n13 = Saturate(None=Multiply(A=Power(Base=@n12) {constExponent=0.5}))
  n14 = Add(
   A=Multiply(A=LinearInterpolate(A=@n10.RGB, B=@n11.RGB, Alpha=@n13))
   B=Multiply(A=VectorParameter() {parameterName=Param_2, defaultValue=<rgba 0.167 0.0169 0.0841 1>}.RGB, B=NamedRerouteUsage())
  ) {constB=0}
  n15 = TextureCoordinate() {coordinateIndex=1}
  n16 = Constant() {r=0.5}
  MaterialFunctionCall(Emissive=LinearInterpolate(A=LinearInterpolate(A=@n9, B=@n14) {constAlpha=0.95}, Alpha=SphereMask(A=@n15, B=@n16) {attenuationRadius=0.2, hardnessPercent=50}) {constB=0}) {materialFunction=&MF_FX_EmissiveAdjustments}
  n17 = TextureCoordinate() {coordinateIndex=1, uTiling=2, vTiling=2}
  n18 = Constant() {r=1}
  n19 = Multiply(
   A=Constant2Vector() {g=1, r=1}
   B=MaterialFunctionCall(Vector or UVs=Subtract(A=@n17, B=@n18)) {materialFunction=&VectorToRadialValue}.Radial Coordinates
  )
  n20 = MaterialFunctionCall(Vector or UVs=Subtract(A=TextureCoordinate() {coordinateIndex=1, uTiling=2, vTiling=2}, B=Constant() {r=1})) {materialFunction=&VectorToRadialValue}
  n21 = Constant2Vector() {g=1, r=1}
  n22 = TextureCoordinate() {coordinateIndex=1}
  n23 = TextureSample!(
   UVs=Panner(Coordinate=Multiply(A=@n20.Radial Coordinates, B=@n21)) {speedX=-0.1}
   Tex
   "DDX(UVs)"=DDX(Value=@n22)
   "DDY(UVs)"=DDY(Value=@n22)
   Apply View MipBias
  ) {
   mipValueMode=TMVM_Derivative
   samplerType=SAMPLERTYPE_LinearGrayscale
   texture=/Engine/Functions/Engine_MaterialFunctions02/ExampleContent/Textures/LowResBlurredNoise
  }
  n24 = Multiply(A=@n23.R) {constB=0.2}
  n25 = TextureCoordinate() {coordinateIndex=1}
  n26 = TextureSample(UVs=Panner(Coordinate=Add(A=@n19, B=@n24)) {speedX=0.2}, "DDX(UVs)"=DDX(Value=@n25), "DDY(UVs)"=DDY(Value=@n25)) {mipValueMode=TMVM_Derivative, samplerType=SAMPLERTYPE_LinearColor, texture=/Game/Effects/Fort_Effects/Emotes/UltraLight/Texture/T_Plasma_D_C}
  n27 = SphereMask(A=TextureCoordinate() {coordinateIndex=1}, B=Constant() {r=0.5}, Radius=Constant() {r=0.5}) {hardnessPercent=0}
  n28 = Multiply(A=Power(Base=@n26.R, Exp=LinearInterpolate(Alpha=@n27) {constA=3, constB=0})) {constB=2}
  n29 = DepthFade(FadeDistance=ScalarParameter() {parameterName=DepthFade})
  n30 = Constant() {r=1}
  n31 = Multiply(
   A=Saturate(None=Multiply(A=@n28) {constB=4})
   B=MaterialFunctionCall("Xb1,PS4"=@n29, "PC-High&&Epic,XSX,PS5"=@n29, "PC-Perf,Android"=@n30, Switch=@n30, "PC-Low(NotPerf)&&Medium"=@n30) {materialFunction=&PlatformSwitch_1D}.Result
  )
  n32 = ScalarParameter() {defaultValue=0.92, parameterName=OpacityMult}
  n33 = ScalarParameter() {defaultValue=60000, parameterName=End}
  n34 = ScalarParameter() {defaultValue=35000, parameterName=Start}
  NamedRerouteDeclaration(None=LinearInterpolate(
    A=Multiply(A=@n31, B=@n32)
    Alpha=MaterialFunctionCall(Fade_Start_Distance=@n33, Fade_End_Distance=@n34) {materialFunction=&MF_Distance_SyncedFade_FX}.Result
   ) {constB=0}) {name=Opacity, nodeColor=<rgba 0 0.6953 1 1>}
  n35 = TextureCoordinate() {coordinateIndex=1, uTiling=2, vTiling=2}
  n36 = Constant() {r=1}
  n37 = Multiply(
   A=Constant2Vector() {g=2, r=1}
   B=MaterialFunctionCall(Vector or UVs=Subtract(A=@n35, B=@n36)) {materialFunction=&VectorToRadialValue}.Radial Coordinates
  )
  n38 = MaterialFunctionCall(Vector or UVs=Subtract(A=TextureCoordinate() {coordinateIndex=1, uTiling=2, vTiling=2}, B=Constant() {r=1})) {materialFunction=&VectorToRadialValue}
  n39 = Constant2Vector() {g=1, r=1}
  n40 = TextureCoordinate() {coordinateIndex=1}
  n41 = TextureSample(UVs=Panner(Coordinate=Multiply(A=@n38.Radial Coordinates, B=@n39)) {speedX=-0.2}, "DDX(UVs)"=DDX(Value=@n40), "DDY(UVs)"=DDY(Value=@n40)) {
   mipValueMode=TMVM_Derivative
   samplerType=SAMPLERTYPE_LinearGrayscale
   texture=&LowResBlurredNoise
  }
  n42 = Multiply(A=@n41.R) {constB=0.2}
  n43 = TextureCoordinate() {coordinateIndex=1}
  n44 = TextureSample(UVs=Panner(Coordinate=Add(A=@n37, B=@n42)) {speedX=0.2}, "DDX(UVs)"=DDX(Value=@n43), "DDY(UVs)"=DDY(Value=@n43)) {mipValueMode=TMVM_Derivative, texture=/Game/Effects/Fort_Effects/Textures/Tile/T_Noise_02}
  n45 = Multiply(
   A=Power(Base=Multiply(A=@n44.R) {constB=30}) {constExponent=3}
   B=SphereMask(A=TextureCoordinate() {coordinateIndex=1}, B=Constant() {r=0.5}, Radius=Constant() {r=0.5}) {hardnessPercent=0}
  )
  n46 = TextureCoordinate() {coordinateIndex=1, uTiling=2, vTiling=2}
  n47 = Constant() {r=1}
  n48 = Multiply(
   A=Constant2Vector() {g=1, r=1}
   B=MaterialFunctionCall(Vector or UVs=Subtract(A=@n46, B=@n47)) {materialFunction=&VectorToRadialValue}.Radial Coordinates
  )
  n49 = MaterialFunctionCall(Vector or UVs=Subtract(A=TextureCoordinate() {coordinateIndex=1, uTiling=2, vTiling=2}, B=Constant() {r=1})) {materialFunction=&VectorToRadialValue}
  n50 = Constant2Vector() {g=1, r=1}
  n51 = TextureCoordinate() {coordinateIndex=1}
  n52 = TextureSample(UVs=Panner(Coordinate=Multiply(A=@n49.Radial Coordinates, B=@n50)) {speedX=-0.2}, "DDX(UVs)"=DDX(Value=@n51), "DDY(UVs)"=DDY(Value=@n51)) {
   mipValueMode=TMVM_Derivative
   samplerType=SAMPLERTYPE_LinearGrayscale
   texture=&LowResBlurredNoise
  }
  n53 = Multiply(A=@n52.R) {constB=0.2}
  n54 = TextureCoordinate() {coordinateIndex=1}
  n55 = TextureSample(UVs=Panner(Coordinate=Add(A=@n48, B=@n53)) {speedX=0.2}, "DDX(UVs)"=DDX(Value=@n54), "DDY(UVs)"=DDY(Value=@n54)) {
   mipValueMode=TMVM_Derivative
   samplerType=SAMPLERTYPE_LinearGrayscale
   texture=/Game/Gadgets/Assets/VinderTech_GliderChute/Glider_Dazzle/FX/Variant/T_Glider_Dazzle_Electric_Crystal_Noise
  }
  n56 = Multiply(A=@n55.R) {constB=3}
  n57 = Multiply(A=Clamp(None=Power(Base=@n56) {constExponent=1}) {maxDefault=5}) {constB=5}
  n58 = OneMinus(None=SphereMask(A=TextureCoordinate() {coordinateIndex=1}, B=Constant() {r=0.5}, Radius=Constant() {r=0.5}) {hardnessPercent=50})
  NamedRerouteDeclaration(None=Multiply(A=LinearInterpolate(A=@n45, B=@n57) {constAlpha=0.3}, B=LinearInterpolate(Alpha=@n58) {constB=0.5})) {name=ColorMask, nodeColor=<rgba 1 0.9844 0 1>}
  NamedRerouteUsage()
 ]
}
