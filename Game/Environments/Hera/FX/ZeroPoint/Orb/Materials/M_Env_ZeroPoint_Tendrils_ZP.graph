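# Translucent unlit material for Zero Point orb energy tendrils on spline meshes and Niagara mesh particles.
# World position offset combines SplineThicken with sine waves and a travelling "zip" bulge, using Index/LastIndex for continuity across segments; MPC_Nexus destabilization speeds the waves.
# Panned star textures distorted by chromatic caustics, plus zip glow, form the emissive.
# Edge, middle, angle, camera and depth fades shape opacity, while distance blends wave, brightness and opacity toward far values.
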Material /Game/Environments/Hera/FX/ZeroPoint/Orb/Materials/M_Env_ZeroPoint_Tendrils_ZP
graph {
 properties {
  bUsedWithNiagaraMeshParticles=true
  bUsedWithSplineMeshes=true
  blendMode=BLEND_Translucent
  pixelDepthOffsetMode=PDOM_Legacy
  shadingModel=MSM_Unlit
  twoSided=true
  usedShadingModels=Unlit
 }
 nodes [
  n1 = NamedRerouteDeclaration(None=ScalarParameter() {defaultValue=250, parameterName=SplineWidth}) {name=SplineWidth, nodeColor=<rgba 0.6172 1 0 1>}
  n2 = ScalarParameter() {defaultValue=1, parameterName=SplineWidthTipMult}
  n3 = ScalarParameter() {defaultValue=1, parameterName=LastIndex}
  n4 = ComponentMask(None=TextureCoordinate()) {g=true}
  n5 = Divide(A=Add(A=ScalarParameter() {parameterName=Index}, B=OneMinus(None=@n4)), B=ScalarParameter() {defaultValue=1, parameterName=LastIndex})
  n6 = Saturate(None=LinearInterpolate(A=Multiply(A=@n3) {constB=5}, Alpha=@n5) {constA=10, constB=0})
  n7 = ScalarParameter() {defaultValue=1, parameterName=LastIndex}
  n8 = Multiply(
   A=Divide(A=ScalarParameter() {defaultValue=0.4, parameterName=ZipRadius}, B=@n7)
   B=LinearInterpolate(
    A=ScalarParameter() {defaultValue=0.4, parameterName=ZipSizeMin}
    B=ScalarParameter() {defaultValue=1.2, parameterName=ZipSizeMax}
    Alpha=NamedRerouteUsage()
   ) {constA=0.4, constB=1.2}
  )
  n9 = Multiply(A=@n8) {constB=0.5}
  n10 = Subtract(B=@n9) {constA=0}
  n11 = Add(A=@n9)
  n12 = Time()
  n13 = Divide(A=ScalarParameter() {defaultValue=3, parameterName=ZipSpeed}, B=@n7)
  n14 = Fmod(A=Add(A=Multiply(A=@n12, B=@n13), B=VertexColor().R), B=Constant() {r=1})
  n15 = Power(Base=SphereMask(A=@n5, B=LinearInterpolate(A=@n10, B=@n11, Alpha=@n14) {constAlpha=1}, Radius=@n8) {attenuationRadius=0.125, hardnessPercent=0})
  n16 = MaterialFunctionCall() {materialFunction=&TangentBasis}
  n17 = LinearInterpolate(
   A=ScalarParameter() {defaultValue=1, parameterName=WaveLength}
   B=ScalarParameter() {defaultValue=1, group=DistanceScaleValue, parameterName=WaveLength_DistanceMax}
   Alpha=NamedRerouteUsage()
  )
  n18 = ScalarParameter() {defaultValue=1, parameterName=LastIndex}
  n19 = Multiply(A=LinearInterpolate(Alpha=VertexColor().R) {constA=0.87, constB=1.125}, B=ScalarParameter() {defaultValue=-0.02, parameterName=WaveSpeed})
  n20 = LinearInterpolate(
   B=ScalarParameter() {defaultValue=3, parameterName=DestabilizeSpeedMult}
   Alpha=CollectionParameter() {collection=/Game/Athena/Environments/Nexus/Materials/FXshader/MPC_Nexus, parameterName=Destabilize2}
  ) {constA=1}
  n21 = Add(A=Divide(A=@n5, B=Divide(A=@n17, B=@n18)), B=Multiply(A=NamedRerouteUsage(), B=Multiply(A=@n19, B=@n20)))
  n22 = Add(A=Multiply(A=VertexColor().R, B=Constant() {r=1}) {constB=2}, B=@n21)
  n23 = NamedRerouteDeclaration(None=ConstantBiasScale(None=Sine(None=@n22) {period=0.7})) {name=Offsets, nodeColor=<rgba 0.9453 0 1 1>}
  n24 = LinearInterpolate(A=@n16.X U, B=@n16.X U, Alpha=@n23)
  n25 = LinearInterpolate(
   A=ScalarParameter() {defaultValue=-1, parameterName=WaveScaleWobbleMin}
   B=ScalarParameter() {defaultValue=1, parameterName=WaveScaleWobbleMax}
   Alpha=ConstantBiasScale(None=Sine(None=@n21))
  )
  n26 = Multiply(
   A=Add(A=Multiply(A=@n15) {constB=0.5}, B=Multiply(A=@n24, B=@n25))
   B=LinearInterpolate(
    A=ScalarParameter() {defaultValue=350, parameterName=WaveScale}
    B=ScalarParameter() {defaultValue=350, group=DistanceScaleValue, parameterName=WaveScale_DistanceMax}
    Alpha=NamedRerouteUsage()
   )
  )
  n27 = ComponentMask(None=TextureCoordinate()) {g=true}
  n28 = ScalarParameter() {defaultValue=1, parameterName=EndFadePower}
  n29 = Constant() {r=0.5}
  n30 = NamedRerouteDeclaration(None=SphereMask(A=Power(Base=@n27, Exp=@n28), B=@n29, Radius=@n29, Hardness=ScalarParameter() {defaultValue=0.64, parameterName=EndFadeHardness}) {attenuationRadius=1, hardnessPercent=50}) {name=EndFade, nodeColor=<rgba 0 1 0.3672 1>}
  Add(
   A=MaterialFunctionCall(WidthBase=@n1, WidthTip=Multiply(A=@n1, B=@n2) {constB=5}) {materialFunction=&SplineThicken}.WorldPositionOffset
   B=Multiply(A=Power(Base=@n6), B=Multiply(A=@n26, B=@n30))
  )
  Constant() {r=0.25}
  n31 = ParticleColor()
  n32 = VectorParameter() {group=DistanceScaleValue, parameterName=EmissiveMult_DistanceMax, defaultValue=<rgba 0.9603 1.3011 1.5 0>}
  n33 = TextureCoordinate()
  n34 = LinearInterpolate(
   A=LinearInterpolate(A=VectorParameter() {parameterName=EmissiveMult, defaultValue=<rgba 0.9603 1.3011 1.5 0>}.RGB, B=@n32.RGB, Alpha=NamedRerouteUsage())
   B=VectorParameter() {parameterName=EmissiveMultDestabilize, defaultValue=<rgba 1.5959 0.6 2 0>}.RGB
   Alpha=Power(Base=ComponentMask(None=@n33) {g=true}, Exp=ScalarParameter() {defaultValue=2, parameterName=EmissiveGradientPower})
  )
  n35 = VectorParameter() {parameterName=Tile1, defaultValue=<rgba 0.4 5 0 0>}
  n36 = Multiply(A=ScalarParameter() {defaultValue=0.005, parameterName=SplineWidthTileMult}, B=NamedRerouteUsage()) {constB=0.1}
  n37 = Multiply(A=TextureCoordinate(), B=AppendVector(A=Multiply(A=@n35.R, B=@n36), B=@n35.G))
  n38 = VectorParameter() {parameterName=Tile2, defaultValue=<rgba 0.5 0.25 0 0>}
  n39 = AppendVector(A=@n38.R, B=@n38.R)
  n40 = Time()
  n41 = ScalarParameter() {defaultValue=0.06, parameterName=PanSpeed}
  n42 = LinearInterpolate(
   B=ScalarParameter() {defaultValue=3, parameterName=DestabilizeSpeedMult}
   Alpha=ScalarParameter() {defaultValue=1, parameterName=Destabilize}
  ) {constA=1}
  n43 = Add(
   A=VertexColor().R
   B=Multiply(A=NamedRerouteDeclaration(None=@n40) {name=Time+Random, nodeColor=<rgba 1 0 0.4922 1>}, B=Multiply(A=@n41, B=@n42))
  )
  n44 = VectorParameter() {parameterName=PanDirection2, defaultValue=<rgba 0.04 0.05 0 0>}
  n45 = TextureSample(UVs=Panner(Coordinate=Multiply(A=@n37, B=@n39), Time=@n43, Speed=AppendVector(A=@n44.R, B=@n44.R)) {speedX=1, speedY=1}) {texture=/Game/Effects/Fort_Effects/Textures/Tile/T_Caustics_Chromatic}
  n46 = Multiply(A=ComponentMask(None=@n45.RGB) {g=true, r=true}) {constB=0.05}
  n47 = VectorParameter() {parameterName=PanDirection1, defaultValue=<rgba 0.08 1 0 0>}
  n48 = AppendVector(A=@n47.R, B=@n47.R)
  n49 = TextureSampleParameter2D(UVs=Panner(Coordinate=Add(A=@n37, B=@n46), Time=@n43, Speed=@n48) {speedX=1, speedY=1}) {parameterName=MainTexture, texture=/Game/Athena/Environments/Nexus/Textures/T_ZeroPoint_Line_Stars}
  n50 = Multiply(A=@n37) {constB=2}
  n51 = VectorParameter() {parameterName=PanDirection1Mult, defaultValue=<rgba -1.7 1.4 0 1>}
  n52 = AppendVector(A=@n51.R, B=@n51.R)
  n53 = TextureSampleParameter2D(UVs=Panner(Coordinate=Subtract(A=@n50, B=@n46), Time=@n43, Speed=Multiply(A=@n48, B=@n52)) {speedX=1, speedY=1}) {parameterName=MainTexture, texture=&T_ZeroPoint_Line_Stars}
  n54 = Add(
   A=@n49.RGB
   B=Multiply(A=@n53.RGB, B=VectorParameter() {parameterName=SecondaryTint, defaultValue=<rgba 0.5521 0.1042 1 1>}.RGB) {constB=0.5}
  )
  n55 = VectorParameter() {parameterName=ZipGlow, defaultValue=<rgba 0 3.0248 5 0>}
  n56 = Constant() {r=-0.1}
  n57 = Constant() {r=1}
  n58 = MaterialFunctionCall(Speed=Constant() {r=8}) {materialFunction=&DebugTimeSine}
  n59 = ScalarParameter() {defaultValue=1, parameterName=UsePulse?}
  n60 = Multiply(
   A=Multiply(A=@n55.RGB) {constB=3}
   B=Multiply(
    A=MaterialFunctionCall!(FlickerInitialSpeed, FlickerTimeOffset=@n56) {materialFunction=/Game/Athena/Environments/Nexus/Pickles/Materials/MF_Pickle_Port_Shudder}.Result
    B=LinearInterpolate(A=@n57, B=@n58.Result, Alpha=@n59)
   ) {constB=0.15}
  ) {constB=10}
  n61 = NamedRerouteUsage()
  n62 = Multiply(A=@n45.RGB, B=Add(A=LinearInterpolate(A=@n60, Alpha=@n61) {constA=5, constB=0}) {constB=0.5}) {constB=0.5}
  n63 = LinearInterpolate(
   A=VectorParameter() {parameterName=BackgroundColor1, defaultValue=<rgba 0.112 0.4078 0.5 0>}.RGB
   B=VectorParameter() {parameterName=BackgroundColor2, defaultValue=<rgba 0.0457 0.0378 0.12 0>}.RGB
   Alpha=@n61
  )
  n64 = TextureCoordinate()
  n65 = Constant() {r=0.5}
  n66 = Multiply(A=NamedRerouteUsage(), B=SphereMask(A=ComponentMask(None=@n64) {r=true}, B=@n65, Radius=@n65) {hardnessPercent=0})
  n67 = MaterialFunctionCall(In=@n66, Contrast=ScalarParameter() {defaultValue=0.25, parameterName=CheapContrast_Contrast}) {materialFunction=&CheapContrast}
  n68 = Add(A=Add(A=Add(A=@n54, B=@n62), B=Multiply(A=@n63, B=@n67.Result)), B=Multiply(A=@n15, B=@n55.RGB))
  MaterialFunctionCall(Emissive=Multiply(A=Multiply(A=@n34, B=@n68), B=@n31.RGB)) {materialFunction=&MF_FX_EmissiveAdjustments}
  n69 = Add(A=Add(A=@n49.B, B=Add(A=@n53.B, B=@n45.B)), B=Multiply(A=@n67.Result) {constB=0.4}) {constB=0.5}
  n70 = Constant() {r=0.5}
  n71 = LinearInterpolate(
   B=ScalarParameter() {parameterName=MiddleFadeMinOpacity}
   Alpha=SphereMask(A=@n5, B=@n70, Radius=@n70, Hardness=ScalarParameter() {defaultValue=0.25, parameterName=MiddleFadeHardness}) {hardnessPercent=0.25}
  ) {constA=1}
  n72 = ScalarParameter() {defaultValue=1, parameterName=AngleFade}
  n73 = Saturate(None=Multiply(A=MaterialFunctionCall("FadeSensitivity (optional)"=@n72) {materialFunction=&FadeBasedOnViewAngle}.Result) {constB=2})
  n74 = Multiply(
   A=MaterialFunctionCall(Distance=ScalarParameter() {defaultValue=800, parameterName=NearFadeDIst}) {materialFunction=&Near_Camera_Fade}.Result
   B=DepthFade(FadeDistance=ScalarParameter() {defaultValue=200, parameterName=DepthFade})
  )
  n75 = Constant() {r=0.7}
  n76 = Multiply(A=Multiply(A=Multiply(A=@n69, B=@n66), B=Power(Base=@n71)), B=ShadingPathSwitch(Default=Multiply(A=@n73, B=@n74), Forward=@n75, Mobile=@n75))
  n77 = Multiply(A=@n15) {constB=0.5}
  n78 = LinearInterpolate(B=LinearInterpolate(A=Max(A=@n77) {constB=0.2}, Alpha=@n23) {constA=0.2}, Alpha=ScalarParameter() {defaultValue=1, parameterName=UsePulse?}) {constA=1}
  n79 = ScalarParameter() {defaultValue=1, parameterName=OpacityMult}
  n80 = ScalarParameter() {defaultValue=1, group=DistanceScaleValue, parameterName=OpacityMult_DistanceMax}
  n81 = NamedRerouteUsage()
  Multiply(A=@n31.A, B=Multiply(A=Multiply(A=@n76, B=@n78), B=LinearInterpolate(A=@n79, B=@n80, Alpha=@n81)))
  NamedRerouteDeclaration(None=MaterialFunctionCall(
    Fade_Start_Distance=ScalarParameter() {defaultValue=80000, group=DistanceScaleValue, parameterName=DistanceFade_End}
    Fade_End_Distance=ScalarParameter() {defaultValue=40000, group=DistanceScaleValue, parameterName=DistanceFade_Start}
   ) {materialFunction=&MF_Distance_SyncedFade_FX}.Result) {name=ScaleValueByDistance, nodeColor=<rgba 1 0.3951 0.9699 1>}
 ]
}
