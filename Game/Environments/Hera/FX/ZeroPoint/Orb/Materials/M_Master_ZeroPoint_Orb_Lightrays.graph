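# Master additive, unlit, two-sided material for the Zero Point orb's light rays, with per-particle randomness, pan speed and tiling from dynamic parameters.
# Two counter-panning line textures, a noise mask and a center gradient form the rays, blended toward a stepped Ceil look and tinted.
# Vertex-color-offset looping time samples punch curves, pulsing brightness and expand/contract offsets atop SkyMango wobble WPO.
# View-angle, depth and optional near-camera fades soften it; time-of-day tint and mobile paths are switchable.
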Material /Game/Environments/Hera/FX/ZeroPoint/Orb/Materials/M_Master_ZeroPoint_Orb_Lightrays
graph {
 properties {
  bAutomaticallySetUsageInEditor=false
  blendMode=BLEND_Additive
  pixelDepthOffsetMode=PDOM_Legacy
  previewMesh=/Hera/Props/SkyMango/SkyMango_Fragment_A/Meshes/SM_SkyMango_Whole
  shadingModel=MSM_Unlit
  twoSided=true
  usedShadingModels=Unlit
 }
 nodes [
  n1 = MaterialFunctionCall!(Tangent Normal, UseWobble=StaticBool()) {materialFunction=&/MF_SkyMango_Whole_WPO}
  n2 = Add(A=@n1.WPO, B=LocalPosition().XYZ)
  n3 = ScalarParameter() {defaultValue=-100, desc=Negative numbers to get smaller, parameterName=Pop Glow WPO Contract}
  n4 = ScalarParameter() {defaultValue=100, parameterName=Pop Glow WPO Expand}
  n5 = TextureCoordinate()
  n6 = ScalarParameter() {defaultValue=5, group=WPO, parameterName=PunchLoopTimeMin}
  n7 = ScalarParameter() {defaultValue=7, group=WPO, parameterName=PunchLoopTimeMax}
  n8 = VertexColor()
  n9 = AppendVector(A=LinearInterpolate(A=@n6, B=@n7, Alpha=@n8.R), B=LinearInterpolate(A=@n6, B=@n7, Alpha=@n8.G))
  n10 = Add(A=Time(), B=Multiply(A=@n9, B=AppendVector(A=@n8.R, B=@n8.R)))
  MaterialExpressionSaturate_0 = Saturate(None=Divide(A=Modulo!(A=@n10, B=@n9), B=ScalarParameter() {defaultValue=3, group=WPO, parameterName=PunchAnimDuration}))
  n11 = Convert(None=@MaterialExpressionSaturate_0) {
   nodeName=Break Float2
   convertInputs [{
     type=Vector2
     expressionInput {
      expression=&self:MaterialExpressionSaturate_0
      inputName=None
      mask=0
      maskA=0
      maskB=0
      maskG=0
      maskR=0
      outputIndex=0
     }
    }]
   convertOutputs [{type=Scalar}, {type=Scalar}]
  }
  n12 = CurveAtlasRowParameter(CurveTime=@n11) {
   atlas=&/CA_SkyMango
   curve=&/SkyMango_Wobble_Punch1
   group=WPO
   parameterName=PunchCurve
  }
  n13 = CurveAtlasRowParameter(CurveTime=@n11) {
   atlas=&CA_SkyMango
   curve=&SkyMango_Wobble_Punch1
   group=WPO
   parameterName=PunchCurve
  }
  n14 = Multiply(A=Step(X=ComponentMask(None=@n5) {g=true}) {constX=0.5, constY=0.5}, B=Max(A=@n12.G, B=@n13.G))
  Add(A=@n1.WPO, B=Multiply(A=Normalize(VectorInput=@n2), B=LinearInterpolate(A=@n3, B=@n4, Alpha=@n14)))
  n15 = NamedRerouteUsage()
  n16 = DynamicParameter() {defaultValue=<rgba 0 1 1 1>, paramNames [Random 0-1, Panner Speed, "U Tile ", Param4]}
  n17 = Multiply(A=@n15, B=@n16."U Tile ")
  n18 = ScalarParameter() {defaultValue=2, parameterName=Gradient Tile A}
  n19 = Multiply(A=@n16.Panner Speed, B=Add(A=@n16.Random 0-1, B=Time()))
  n20 = TextureObjectParameter() {parameterName="Lines Texture (RG)", texture=/Game/Effects/Fort_Effects/Textures/Outlander/T_InTheZone_LightRays_01}
  n21 = TextureSample(UVs=Panner(Coordinate=Multiply(A=@n17, B=@n18), Time=@n19) {speedX=0.0702}, Tex=@n20) {texture=&T_InTheZone_LightRays_01}
  n22 = Multiply(A=@n16."U Tile ", B=@n15)
  n23 = ScalarParameter() {defaultValue=2.5, parameterName=Gradient Tile B}
  n24 = TextureSample(UVs=Panner(Coordinate=Multiply(A=@n22, B=@n23), Time=@n19) {speedX=-0.047}, Tex=@n20) {texture=&T_InTheZone_LightRays_01}
  n25 = TextureCoordinate() {uTiling=0.33, vTiling=0.01}
  n26 = LinearInterpolate(A=@n25, B=OneMinus(None=@n25), Alpha=ScalarParameter() {parameterName=InvertUV})
  n27 = DynamicParameter() {defaultValue=<rgba 0 1 1 1>, paramNames [Random 0-1, Panner Speed, "U Tile ", Param4]}
  n28 = TextureSample(
   UVs=Panner(Coordinate=Add(A=@n26, B=@n27.Random 0-1), Time=@n19) {speedX=0.0436}
   Tex=TextureObjectParameter() {parameterName="Lines Texture Mask (R)", texture=&T_NoisePack_01}
  ) {texture=&T_NoisePack_01}
  n29 = ScalarParameter() {defaultValue=3, parameterName=Mask Brightness}
  n30 = ScalarParameter() {defaultValue=-0.03, parameterName=Slide}
  n31 = OneMinus(None=ComponentMask(None=Add(A=@n15, B=@n30)) {g=true})
  n32 = ScalarParameter() {defaultValue=12, parameterName=Center Exponent}
  n33 = Add(
   A=Multiply(A=Multiply(A=@n21.R, B=@n24.G), B=Multiply(A=@n28.R, B=@n29))
   B=Multiply(A=Power(Base=@n31, Exp=@n32), B=ScalarParameter() {defaultValue=2, parameterName=Center Brightness})
  )
  n34 = Add(A=@n33, B=ScalarParameter() {defaultValue=-0.7, parameterName=Ceil Bias})
  n35 = LinearInterpolate(
   A=@n33
   B=Multiply(A=Ceil(None=@n34), B=ScalarParameter() {defaultValue=3, parameterName=Ceil Brightness})
   Alpha=ScalarParameter() {defaultValue=0.5, parameterName=Style Lerp}
  )
  n36 = ComponentMask(None=TextureCoordinate()) {g=true}
  n37 = OneMinus(None=Power(Base=OneMinus(None=@n36)) {constExponent=16})
  n38 = OneMinus(None=ComponentMask(None=NamedRerouteUsage()) {g=true})
  n39 = TextureCoordinate() {uTiling=3, vTiling=0}
  n40 = LinearInterpolate(A=@n39, B=OneMinus(None=@n39), Alpha=ScalarParameter() {parameterName=InvertUV})
  n41 = LinearInterpolate(Alpha=TextureSample(UVs=Panner(Coordinate=@n40) {speedX=0.2, speedY=0.1}) {texture=&T_NoisePack_01}.R) {constA=2, constB=6}
  n42 = Multiply(
   A=LinearInterpolate(
    A=VectorParameter() {parameterName=Color_01, defaultValue=<rgba 1.9601 0.0218 2 1>}.RGB
    B=VectorParameter() {parameterName=Color_02, defaultValue=<rgba 0 0 1 1>}.RGB
   )
   B=Multiply(A=Multiply(A=@n35, B=@n37), B=Power(Base=@n38, Exp=@n41))
  )
  n43 = Add(A=@n12.G, B=@n13.G)
  Multiply(
   A=MaterialFunctionCall(Emissive=@n42) {materialFunction=&MF_FX_EmissiveAdjustments}.Result
   B=LinearInterpolate(
    A=ScalarParameter() {parameterName=Pop Idle Brightness}
    B=ScalarParameter() {defaultValue=0.5, parameterName=Pop Glow Brightness}
    Alpha=VertexInterpolator(VS=@n43).PS
   ) {constA=1, constB=2}
  )
  n44 = VectorParameter() {parameterName=RayBlend, defaultValue=<rgba 16 0.3 0.8 1>}
  n45 = DotProduct(A=CameraVectorWS(), B=PixelNormalWS())
  Multiply(A=@n37, B=SmoothStep(Min=@n44.G, Max=@n44.G, Value=Abs(None=@n45)) {constMax=0.8, constMin=0.3})
  n46 = TextureCoordinate()
  NamedRerouteDeclaration(None=LinearInterpolate(A=@n46, B=OneMinus(None=@n46), Alpha=ScalarParameter() {parameterName=InvertUV})) {name=UV, nodeColor=<rgba 0 0.0625 1 1>}
  n47 = ScalarParameter() {defaultValue=1, parameterName=Color_Day}
  n48 = ScalarParameter() {defaultValue=1, parameterName=Color_Night}
  n49 = MaterialFunctionCall(Color_Morning=@n47, Color_Day=@n47, Color_Evening=@n48, Color_Night=@n48) {materialFunction=&Color_Time_of_Day}
  Reroute(None=StaticSwitchParameter(True=@n42, False=Multiply(A=@n42, B=@n49.Result)) {defaultValue=true, parameterName=Color of day})
  n50 = DepthFade(FadeDistance=ScalarParameter() {defaultValue=50, parameterName=Depth Fade Distance})
  n51 = QualitySwitch(Default=@n50, Low=Constant() {r=1}, High=@n50, Medium=@n50)
  StaticSwitchParameter(True=ShadingPathSwitch(Default=@n50, Mobile=Constant() {r=1}) {desc=Migrated from FeatureLevelSwitch}, False=@n51) {parameterName=Use_MobilePerf_Settings}
  n52 = ScalarParameter() {defaultValue=250, parameterName=NearCamFade_Distance}
  n53 = Multiply(B=@n51) {constA=1}
  StaticSwitchParameter(
   True=Multiply(A=MaterialFunctionCall(Distance=@n52) {materialFunction=&Near_Camera_Fade}.Result, B=@n53)
   False=@n53
  ) {parameterName=Use_NearCamFade?}
  TextureCoordinate()
  TextureCoordinate()
  TextureCoordinate()
 ]
}
