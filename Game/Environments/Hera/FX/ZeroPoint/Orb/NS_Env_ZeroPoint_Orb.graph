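# Niagara environment effect for the Zero Point orb, with three emitters whose spawn rates scale with User.Debris_SpawnRate.
# GPU mesh Debris spawns at hard-coded positions and drifts inward, with region masks driving acceleration, spin, size and glow color.
# Sprites spawn around a sphere with inward velocity, curl noise and flickering color.
# Tendrils are normal-aligned meshes that grow outward, fading by time of day; all emitters use distance culling and distance-based scaling.
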NiagaraSystem /Game/Environments/Hera/FX/ZeroPoint/Orb/NS_Env_ZeroPoint_Orb
dependencies {
 usedDataInterfaces [
  &NiagaraDataInterfaceCurve
  &NiagaraDataInterfaceColorCurve
  &NiagaraDataInterfaceArrayFloat3
  &NiagaraDataInterfaceCamera
  &NiagaraDataInterfaceVectorCurve
 ]
 usedDynamicInputs [&Add_Float, &ModuloFloat, &Sine]
 usedModules [
  &SystemState
  &self:SystemSpawnScript.SystemScriptSource.SystemScriptGraph.NiagaraNodeAssignment_0.SetVariables_#1
 ]
}
emitters {
 Debris {
  simTarget=GPUComputeSim
  emitterUpdateScript [
   &EmitterState(Enable Distance Culling=true, MaxDistance=25000, Module.bUseMaxDistance=true, Module.bUseMinDistance=true, Scalability Mode=::NewEnumerator1)
   &SpawnRate(SpawnRate=&Multiply_Float(A=110, B=$User.Debris_SpawnRate), Use Spawn Probability=false)
  ]
  particleSpawnScript [
   /SetVariables(Particles.MeshIndex=&RandomRangeInt(Minimum=0, Maximum=2)) {isSetParameters=true}
   &InitializeParticle(
    Color Mode=::NewEnumerator0
    Lifetime=3
    Mesh Scale Mode=::NewEnumerator3
    Mesh Uniform Scale=&RandomRangeFloat(Minimum=1, Maximum=1.5)
    Sprite Size Mode=::NewEnumerator3
    Uniform Sprite Size=500
    UsePositionOffset=false
    Write Material Random=true
   )
   /SetVariables(Particles.Position=&AddVectorToPosition(
     Position=&SelectVectorFromArray(
      Array Sampling Mode=::NewEnumerator0
      Direct Array Index=?
      Vector Selection Array=DataInterface() {propertyValues {FloatData [
         <761.1786 -514.6325 25.523>
         <782.1675 -182.1788 583.1962>
         <789.4163 -470.6479 352.0369>
         <606.762 -499.2092 678.5782>
         <1016.7694 -175.3458 487.9492>
         <930.5601 -388.2798 144.1381>
         <1041.5487 -390.0174 576.2992>
         <1365.8231 -37.881 452.5173>
         <1404.6393 -132.9956 384.1602>
         <1533.0074 -291.0304 539.7721>
         <1368.8838 104.3543 413.3715>
         <1511.6901 28.1403 473.7091>
         <2019.0536 122.8359 459.8264>
         <1872.5118 138.0785 447.5024>
         <2012.9375 190.344 580.0704>
         <2049.0808 160.3047 324.1908>
         <2405.9485 181.3052 509.3727>
         <2322.6528 271.9438 358.9022>
         <2617.1206 259.0288 464.646>
         <2569.0554 300.5549 341.5598>
         <2791.3254 161.4196 315.7773>
         <2638.0171 143.846 362.1083>
         <2878.8037 129.5937 381.1584>
         <3018.4026 267.1469 288.6014>
         <3232.2932 259.4214 577.9879>
         <3316.6025 4.1469 400.6093>
         <3391.8362 -65.3556 515.4407>
         <3287.4731 82.2751 668.7482>
         <-614.9214 624.9576 117.1163>
         <-1519.7988 417.7064 -160.5957>
         <-1179.1381 38.3141 352.5999>
         <-1143.2274 773.0903 356.0123>
         <-1004.821 801.5428 259.2526>
         <-1068.9534 208.6137 394.7797>
         <-2011.0295 29.2948 403.1727>
         <-1355.0601 30.1 513.5507>
         <-1923.455 -241.1316 181.42>
         <-2087.3513 -171.1794 741.8907>
         <-1878.3481 -66.7814 339.5296>
         <-2217.6594 -139.6226 304.599>
         <-2308.2302 -161.2984 536.1887>
         <-2139.5549 -288.6132 640.3519>
         <-2831.5356 -521.1422 513.1672>
         <-2521.3203 -364.3717 389.4>
         <-2625.7727 -475.3568 317.0504>
         <-2705.72 -371.4984 573.2266>
         <-3296.1887 -387.7885 666.4712>
         <-3064.2935 -335.995 659.6409>
         <-3399.0769 -643.7805 741.9924>
         <-3452.5842 -737.2614 660.6478>
         <-3596.519 -458.8765 782.9475>
         <-3827.116 -603.84 867.9534>
         <-4021.7471 -621.7287 887.5188>
         <-3918.8555 -609.931 1018.4387>
         <-4269.8701 -733.4377 1145.5262>
         <-4311.4292 -700.7148 1110.366>
         <794.2189 -857.3174 1018.3815>
         <-667.0729 -640.996 969.073>
         <-517.5557 -1007.0194 748.2405>
         <-236.1603 -832.3739 773.7617>
         <-269.5095 -935.9778 740.0543>
         <-316.7804 -1089.0378 853.1439>
         <-684.6012 -1042.8699 1513.5546>
         <-312.1656 -1124.0178 1174.4271>
         <-844.2437 -1864.8818 1311.8828>
         <-431.3018 -1827.6641 1362.8534>
         <-566.0699 -2238.9541 1703.2974>
         <-778.4214 -2179.2271 1440.9677>
         <-652.2798 -2514.6335 1513.8149>
         <-962.5885 -2507.5688 1768.885>
         <-705.5828 -2975.187 1737.3311>
         <-696.8292 -3339.7246 1871.1416>
         <-905.3392 -3340.6243 1876.6676>
         <-1198.2806 -3238.8938 1702.2511>
         <-979.0031 -3482.6838 1837.4041>
         <-1006.091 -3621.0686 1870.2355>
         <-952.9475 -3763.28 1959.0964>
         <-995.232 -3906.5291 2040.646>
         <-1127.1364 -4315.3828 1859.4136>
         <-871.811 -4051.6072 1993.0317>
         <-1196.0502 -4664.3647 2192.3765>
         <-1182.0729 -4680.0444 1969.4202>
         <-1190.2216 -4922.5503 2355.8391>
         <-1313.5874 -4789.0049 2222.103>
         <605.6788 -1107.2095 -342.5872>
         <302.4471 -1089.4719 99.8487>
         <190.8519 -863.6199 208.9009>
         <161.8813 -1462.4731 -240.0787>
         <562.5263 -1142.2059 -102.6544>
         <371.4693 -886.6703 76.7946>
         <652.1164 -676.1639 -101.8865>
         <1011.1004 -532.6637 -340.9526>
         <844.7271 -1408.1138 -102.708>
         <1046.1129 -1285.4653 -448.4518>
         <769.4218 -1211.0645 67.7661>
         <680.4373 -1139.389 15.0663>
         <1050.0176 -1132.1859 -398.5455>
         <753.0139 -1287.3709 -169.7625>
         <971.5989 -1296.2784 -137.7734>
         <936.4014 -1433.2272 -271.5787>
         <1057.3979 -1274.8831 -168.3365>
         <1052.333 -1303.7528 -335.5756>
         <1117.954 -1678.7029 -188.3011>
         <1199.1561 -1595.9453 -228.7505>
         <1443.0815 -1833.965 -290.1758>
         <1327.5055 -1654.8983 -48.6098>
         <1474.9966 -1804.103 -312.3262>
         <1371.651 -1592.1517 -244.2037>
         <1416.6912 -1764.2609 -510.2405>
         <1502.8318 -1817.1646 -347.2819>
         <1443.6099 -1997.2721 -400.5905>
         <1541.3223 -1958.0709 -423.9521>
         <-441.8069 826.7213 480.9047>
         <-4.1697 894.2682 911.6484>
         <-296.4847 551.9293 1456.0557>
         <-196.5987 774.9719 1032.5171>
         <349.7916 871.8864 1358.2483>
         <-38.996 704.1675 1398.1587>
         <-480.7733 1321.5729 1857.7778>
         <-310.4426 1391.3304 1816.4902>
         <-172.4787 1306.7343 1535.946>
         <-751.8242 1695.7048 1628.6548>
         <-713.8727 1837.8057 2405.1162>
         <-725.0181 1547.7797 2216.8655>
         <-571.2446 1877.4214 2195.436>
         <-1031.3538 1840.7396 2319.5115>
         <-900.9855 2020.0514 2535.7681>
         <-791.1448 1576.7748 2441.7522>
         <-671.9496 2314.4338 2774.4775>
         <-1129.8635 2253.9758 2606.5474>
         <-890.6725 2683.6597 2544.2217>
         <-854.6989 2572.4177 2653.6553>
         <-908.4916 2968.49 3017.8149>
         <-890.5733 2937.1421 2789.563>
         <-918.5649 3218.6631 3108.999>
         <-920.9218 3275.5791 3069.0161>
         <-1168.9994 3506.2256 3023.7483>
         <-1062.7303 3445.6565 3083.2432>
         <-896.514 3781.5142 3156.2109>
         <-889.8562 3867.0193 3098.4841>
        ]}}
      Random Seed=0
     )
     Vector=<0 0 333>
    )) {isSetParameters=true}
   &ShapeLocation(MaskSpawnGroup=false, Override Seed=false, Sphere Radius=1, UseOwnerScale=true)
   &AddVelocity(
    Constrain To Radius=2000
    Module.Use Falloff Distance=false
    Origin Offset=&RandomRangeVector(Minimum=<-888 -888 -888>, Maximum=<888 888 888>)
    Override Seed=false
    Velocity Mode=::NewEnumerator1
    Velocity Speed=-222
   )
   &InitialMeshOrientation(Use Rotation Vector=false)
  ]
  particleUpdateScript [
   &ParticleState()
   &Drag(Use Linear Drag=true, Use Rotational Drag=true)
   &RegionMask(Region Falloff=1, Region Radius=4000, UseOffset=false, UseRotation=false)
   &RegionMask(Region Radius=3000, UseOffset=false, UseRotation=false) {module=RegionMask001}
   &ScaleMeshSize(Scale Factor=&Multiply_VectorByFloat(
     Vector=&VectorFromFloat(Value=&FloatFromCurve(
       FloatCurve=DataInterface() {propertyValues {Curve {
          defaultValue=1
          keys [
           {interpMode=RCIM_Cubic, tangentMode=RCTM_User, time=0, value=0}
           {interpMode=RCIM_Cubic, tangentMode=RCTM_User, time=0.2, value=1}
           {arriveTangent=-0.3456, interpMode=RCIM_Cubic, leaveTangent=-1.25, tangentMode=RCTM_Break, time=0.8036, value=1}
           <at 1 0 RCIM_Cubic>
          ]
         }}}
       Scale Curve=&Lerp_Float(
        A=0.5
        B=1.3
        Alpha=&FloatFromCurve(
         FloatCurve=DataInterface() {propertyValues {Curve {
            defaultValue=1
            keys [
             {interpMode=RCIM_Cubic, tangentMode=RCTM_User, time=0, value=0}
             {interpMode=RCIM_Cubic, leaveTangent=-2.7545, tangentMode=RCTM_Break, time=0.3, value=1}
             <at 0.8 0 RCIM_Cubic>
            ]
           }}}
         CurveIndex=$Particles.RegionMask.Float
        )
       )
      ))
     Float=&FloatFromCurve(
      FloatCurve=DataInterface() {propertyValues {Curve {
         defaultValue=0
         keys [{interpMode=RCIM_Cubic, tangentMode=RCTM_User, time=0, value=1}, {interpMode=RCIM_Cubic, tangentMode=RCTM_User, time=1, value=2}]
        }}}
      CurveIndex=$Emitter.Scalability.DistanceFraction
     )
    )) {enabled=false, module=ScaleMeshSize002}
   &ScaleMeshSize(Scale Factor=&Multiply_VectorByFloat(
     Vector=&VectorFromFloat(Value=&FloatFromCurve(
       FloatCurve=DataInterface() {propertyValues {Curve {
          defaultValue=1
          keys [
           {interpMode=RCIM_Cubic, tangentMode=RCTM_User, time=0, value=0}
           {interpMode=RCIM_Cubic, tangentMode=RCTM_User, time=0.2, value=1}
           {arriveTangent=-0.3456, interpMode=RCIM_Cubic, leaveTangent=-1.25, tangentMode=RCTM_Break, time=0.8036, value=1}
           <at 1 0 RCIM_Cubic>
          ]
         }}}
       Scale Curve=&Lerp_Float(
        A=0.5
        B=1.3
        Alpha=&FloatFromCurve(
         FloatCurve=DataInterface() {propertyValues {Curve {
            defaultValue=1
            keys [
             {interpMode=RCIM_Cubic, tangentMode=RCTM_User, time=0, value=0}
             {interpMode=RCIM_Cubic, leaveTangent=-2.7545, tangentMode=RCTM_Break, time=0.3, value=1}
             <at 0.8 0 RCIM_Cubic>
            ]
           }}}
         CurveIndex=$Particles.RegionMask.Float
        )
       )
      ))
     Float=&Multiply_Float(
      A=&FloatFromCurve(
       FloatCurve=DataInterface() {propertyValues {Curve {
          defaultValue=0
          keys [{interpMode=RCIM_Cubic, tangentMode=RCTM_User, time=0, value=1}, {interpMode=RCIM_Cubic, tangentMode=RCTM_User, time=1, value=2}]
         }}}
       CurveIndex=$Emitter.Scalability.DistanceFraction
      )
      B=&FloatFromCurve(
       FloatCurve=DataInterface() {propertyValues {Curve {
          defaultValue=0
          keys [
           {interpMode=RCIM_Cubic, tangentMode=RCTM_User, time=0.2, value=0}
           {arriveTangent=0.0129, interpMode=RCIM_Cubic, leaveTangent=0.0129, tangentMode=RCTM_User, time=0.5, value=1}
          ]
         }}}
       CurveIndex=/Game/Effects/Niagara/Modules/Math/FN_CameraDistance(
        0Dist:NiagaraFloat=0
        1Dist:NiagaraFloat=8000
        Camera Query:NiagaraDataInterfaceCamera=DataInterface() {propertyValues {bRequireCurrentFrameData=false}}
        Position:NiagaraPosition=$Particles.Position
       )
      )
     )
    ))
   /SetVariables() {isSetParameters=true}
   &UpdateMeshOrientation(
    Rotation Rate=&Lerp_Float(A=0.05, B=&RandomRangeFloat(Minimum=0.2, Maximum=0.2), Alpha=$Particles.RegionMask001.Float)
    Rotation Vector=&RandomRangeVector(Minimum=<0 0 0>, Maximum=<1 1 1>)
   )
   &AccelerationForce(Acceleration=&Multiply_Vector(
     A=&Multiply_Vector(A=$Particles.Velocity, B=&VectorFromFloat(Value=$Output.RegionMask.RegionMask))
     B=&VectorFromFloat(Value=0.4)
    ))
   /SetVariables(Particles.DecalGlow_Color:LinearColor=&MultiplyLinearColor(
     Linear Color=&LerpLinearColors(
      StartColor=<rgba 0.1999 0.1999 2 1>
      EndColor=<rgba 0.9229 0.2287 2 1>
      LerpFactor=&FloatFromCurve(
       FloatCurve=DataInterface() {propertyValues {Curve {
          defaultValue=0
          keys [{interpMode=RCIM_Cubic, tangentMode=RCTM_User, time=0.1, value=0}, {interpMode=RCIM_Cubic, tangentMode=RCTM_User, time=0.5, value=1}]
         }}}
       CurveIndex=$Particles.RegionMask.Float
      )
     )
     Scale Factor=&MakeVector4(
      X=1
      Y=1
      Z=1
      W=&Waveform(
       "Waveform [1]"=::NewEnumerator0
       "[1] Phase Scale"=&RandomRangeFloat(Minimum=0.5, Maximum=0.1)
       "[1] Frequency"=1
       "[1] Amplitude Min/Max"=&RandomRangeVector2D(Minimum=<0 1>, Maximum=<0.2 1>)
       "[1] Scale"=&RandomRangeFloat(Minimum=0.15, Maximum=0.6)
      )
     )
    )) {isSetParameters=true}
   &SolveForcesAndVelocity(Clamp Velocity=false, Limit Acceleration=false)
   &CameraOffset(Camera Offset Amount=&RandomRangeFloat(Minimum=22, Maximum=-22))
  ]
  rendererClasses [&NiagaraMeshRendererProperties]
  renderers [{rendererClass=&NiagaraMeshRendererProperties, rendererIndex=0}]
 }
 Sprites {
  emitterUpdateScript [
   &EmitterState(
    Enable Distance Culling=true
    MaxDistance=35000
    Module.bUseMaxDistance=true
    Module.bUseMinDistance=false
    Scalability Mode=::NewEnumerator1
   )
   &SpawnRate(
    SpawnRate=&Multiply_Float(A=&RandomRangeFloat(Minimum=60, Maximum=99), B=$User.Debris_SpawnRate)
    Use Spawn Probability=false
   )
  ]
  particleSpawnScript [
   &InitializeParticle(
    Color Mode=::NewEnumerator0
    Lifetime Min=0.5
    Lifetime Mode=::NewEnumerator1
    Sprite Size Mode=::NewEnumerator3
    Uniform Sprite Size=&Multiply_Float(A=&RandomRangeFloat(Minimum=44, Maximum=11), B=4)
    UsePositionOffset=false
    Write Material Random=true
   )
   &ShapeLocation(MaskSpawnGroup=false, Override Seed=false, Sphere Radius=2000, Sphere Surface Distribution=1, UseOwnerScale=true)
   &AddVelocity(Module.Use Falloff Distance=false, Override Seed=false, Velocity Mode=::NewEnumerator1, Velocity Speed=-999)
  ]
  particleUpdateScript [
   &ParticleState()
   &CurlNoiseForce(Bias Noise=false, Noise Strength=1092.5066, Pan Noise=false)
   &Drag(
    Drag=&FloatFromCurve(
     FloatCurve=DataInterface() {propertyValues {Curve {
        defaultValue=1
        keys [{interpMode=RCIM_Cubic, tangentMode=RCTM_User, time=0, value=0.3}, {interpMode=RCIM_Cubic, tangentMode=RCTM_User, time=1, value=1}]
       }}}
     Scale Curve=2
    )
    Use Linear Drag=true
    Use Rotational Drag=true
   )
   &SolveForcesAndVelocity(Clamp Velocity=false, Limit Acceleration=false)
   &ScaleColor(
    Scale Alpha=&Multiply_Float(
     A=&FloatFromCurve(
      FloatCurve=DataInterface() {propertyValues {Curve {
         defaultValue=1
         keys [
          <at 0 0 RCIM_Cubic>
          {arriveTangent=1.6667, interpMode=RCIM_Cubic, leaveTangent=0.5618, tangentMode=RCTM_Break, time=0.2, value=1}
          {interpMode=RCIM_Cubic, tangentMode=RCTM_User, time=0.6, value=1}
          {interpMode=RCIM_Cubic, tangentMode=RCTM_User, time=1, value=0}
         ]
        }}}
      Scale Curve=&Waveform(
       "Waveform [1]"=::NewEnumerator0
       "[1] Frequency"=&RandomRangeFloat(Minimum=7, Maximum=9)
       "[1] Amplitude Min/Max"=<0.5 1>
       "[1] Scale"=&RandomRangeFloat(Minimum=0.3, Maximum=0.5)
      )
     )
     B=2
    )
    Scale RGB=&MakeVectorFromLinearColorRGB(Color=&ColorFromCurve(ColorCurve=DataInterface() {propertyValues {
        AlphaCurve {keys [<at 0.3412 1 RCIM_Cubic>,<at 1 0 RCIM_Cubic>]}
        BlueCurve {keys [<at 0 2.1349 RCIM_Cubic>,<at 0.1791 6>,<at 0.3996 0.7693>,<at 1 1 RCIM_Cubic>]}
        GreenCurve {keys [<at 0 0.3266 RCIM_Cubic>,<at 0.1791 1.6971>,<at 0.3996 0.2871>,<at 1 0.0675 RCIM_Cubic>]}
        RedCurve {keys [<at 0 3 RCIM_Cubic>,<at 0.1791 5.5827>,<at 0.3996 1>,<at 1 0.3725 RCIM_Cubic>]}
       }}))
    ScaleA=true
    ScaleRGB=true
   )
   &CameraOffset(Camera Offset Amount=&RandomRangeFloat(Minimum=0, Maximum=9))
   /SetVariables(
    Particles.GlowSharp_Color:LinearColor=&MultiplyLinearColor(
     Linear Color=&ColorFromCurve(ColorCurve=DataInterface() {propertyValues {
        AlphaCurve {keys [<at 0.3412 1 RCIM_Cubic>,<at 1 0 RCIM_Cubic>]}
        BlueCurve {keys [<at 0 2.1349 RCIM_Cubic>,<at 0.1791 6>,<at 0.3996 0.7693>,<at 1 1 RCIM_Cubic>]}
        GreenCurve {keys [<at 0 0.3266 RCIM_Cubic>,<at 0.1791 1.6971>,<at 0.3996 0.2871>,<at 1 0.0675 RCIM_Cubic>]}
        RedCurve {keys [<at 0 3 RCIM_Cubic>,<at 0.1791 5.5827>,<at 0.3996 1>,<at 1 0.3725 RCIM_Cubic>]}
       }})
     Scale Factor=/Niagara/DynamicInputs/UniformRange/V2/RandomRangeVector4(
      Minimum:Vector4f=<333 1 333 1>
      Maximum:Vector4f=<1 1 1 0>
      Evaluation Type:ENiagaraRandomnessEvaluation=::NewEnumerator0
      Recalculate Random Each Loop:NiagaraBool=true
      Randomness Mode:ENiagaraRandomnessMode=::NewEnumerator0
      Random Seed:NiagaraInt32=0
      Fixed Random Seed:NiagaraBool=false
     )
    )
    Particles.GlowSharp_Scale:Vector2f=&Multiply_Vector2D_ByFloat(
     Float=&FloatFromCurve(
      FloatCurve=DataInterface() {propertyValues {Curve {
         defaultValue=1
         keys [
          {interpMode=RCIM_Cubic, tangentMode=RCTM_User, time=0, value=0}
          {interpMode=RCIM_Cubic, tangentMode=RCTM_User, time=0.3, value=1}
          {arriveTangent=-0.4712, interpMode=RCIM_Cubic, leaveTangent=-1.4289, tangentMode=RCTM_Break, time=0.7, value=1}
          <at 0.9998 0 RCIM_Cubic>
         ]
        }}}
      Scale Curve=22
     )
     Vector2D=<1 1>
    )
   ) {isSetParameters=true}
   &ScaleSpriteSize(
    Uniform Curve Index=$Emitter.Scalability.DistanceFraction
    Uniform Curve Sprite Scale=DataInterface() {propertyValues {Curve {
       defaultValue=1
       keys [
        <at 0 1 RCIM_Cubic>
        {arriveTangent=-0.1106, interpMode=RCIM_Cubic, leaveTangent=1, tangentMode=RCTM_Break, time=0.5, value=1}
        {tangentMode=RCTM_User, time=1, value=2.5}
       ]
      }}}
   )
  ]
  rendererClasses [&NiagaraSpriteRendererProperties]
  renderers [
   {rendererClass=&NiagaraSpriteRendererProperties}
   {rendererClass=&NiagaraSpriteRendererProperties, rendererIndex=1}
   {rendererClass=&NiagaraSpriteRendererProperties, rendererIndex=2}
  ]
 }
 Tendrills_Small {
  emitterUpdateScript [
   &EmitterState(Enable Distance Culling=true, MaxDistance=35000, Module.bUseMaxDistance=true, Module.bUseMinDistance=true, Scalability Mode=::NewEnumerator1)
   &SpawnRate(SpawnRate=&Multiply_Float(A=4, B=$User.Debris_SpawnRate), Use Spawn Probability=false)
  ]
  particleSpawnScript [
   &InitializeParticle(
    Color Maximum=<rgba 0.9393 0.5491 1 1>
    Color Minimum=<rgba 0.3822 0.1567 2 1>
    Color Mode=::NewEnumerator2
    Lifetime Max=4
    Lifetime Min=3
    Lifetime Mode=::NewEnumerator1
    Mesh Scale=&Multiply_VectorByFloat(Vector=&RandomRangeVector(Minimum=<15 15 15>, Maximum=<30 30 30>), Float=0.5)
    Mesh Scale Mode=::NewEnumerator4
    UsePositionOffset=false
    Write Material Random=true
   )
   &ShapeLocation(MaskSpawnGroup=false, Override Seed=false, Sphere Radius=1100, Sphere Surface Distribution=1, UseOwnerScale=true)
   &InitialMeshOrientation(
    Mesh Axis To Orient=&Multiply_Vector(A=$Particles.ShapeLocation.ShapeNormal, B=<1 -1 -1>)
    Mesh Orientation Mode=::NewEnumerator2
    Use Rotation Vector=false
   ) {module=InitialMeshOrientation001}
   &AddVelocity(Module.Use Falloff Distance=false, Override Seed=false, Velocity Mode=::NewEnumerator1, Velocity Speed=55)
  ]
  particleUpdateScript [
   &ParticleState()
   &SolveForcesAndVelocity(Clamp Velocity=false, Limit Acceleration=false)
   &DynamicMaterialParameters(
    Index 0 Param 1=&RandomRangeFloat(Minimum=0, Maximum=1)
    Index 0 Param 2=&RandomRangeFloat(Minimum=2, Maximum=4)
    Index 0 Param 4=&FloatFromCurve(
     FloatCurve=DataInterface() {propertyValues {Curve {defaultValue=1, keys [{interpMode=RCIM_Cubic, tangentMode=RCTM_User, time=0.25, value=0}, <at 1 1 RCIM_Cubic>]}}}
     Scale Curve=3
    )
    Module.Index0Float0Write=true
    Module.Index0Float1Write=true
    Module.Index0Float2Write=true
    Module.Index0Float3Write=true
   )
   /SetVariables(Particles.DebugScale:Vector3f=&VectorFromFloat(Value=2)) {isSetParameters=true}
   &ScaleColor(
    Scale Alpha=&Multiply_Float(
     A=&FloatFromCurve(
      FloatCurve=DataInterface() {propertyValues {Curve {
         defaultValue=1
         keys [
          {arriveTangent=6.4544, interpMode=RCIM_Cubic, time=0, value=0}
          {
           arriveTangent=1.5375
           arriveTangentWeight=0.1333
           interpMode=RCIM_Cubic
           leaveTangent=1.5375
           leaveTangentWeight=0.0667
           tangentMode=RCTM_User
           time=0.2845
           value=0.4847
          }
          {
           arriveTangent=-1.9657
           arriveTangentWeight=0.1296
           interpMode=RCIM_Cubic
           leaveTangent=-1.9657
           leaveTangentWeight=0.2592
           tangentMode=RCTM_User
           time=0.6793
           value=0.8725
          }
          {tangentMode=RCTM_User, time=1, value=0}
         ]
        }}}
      Scale Curve=2
     )
     B=&FloatFromCurve(
      FloatCurve=DataInterface() {propertyValues {Curve {
         defaultValue=1
         keys [
          {interpMode=RCIM_Cubic, tangentMode=RCTM_User, time=0, value=0.55}
          {interpMode=RCIM_Cubic, tangentMode=RCTM_User, time=0.5, value=1}
          {tangentMode=RCTM_User, time=1, value=0.55}
         ]
        }}}
      CurveIndex=Linked() {linkedVariable {name=NPC.FortniteNPC.TimeOfDay_0-1, type=NiagaraFloat}}
     )
    )
    Scale RGB=&Multiply_VectorByFloat(
     Vector=<1 1 1>
     Float=&FloatFromCurve(FloatCurve=DataInterface() {propertyValues {Curve {defaultValue=1, keys [{interpMode=RCIM_Cubic, tangentMode=RCTM_User, time=0, value=6}, <at 0.7 2 RCIM_Cubic>]}}})
    )
    ScaleA=true
    ScaleRGB=true
   ) {enabled=false, module=ScaleColor001}
   &ScaleColor(
    Scale Alpha=&Multiply_Float(
     A=&Multiply_Float(
      A=&FloatFromCurve(
       FloatCurve=DataInterface() {propertyValues {Curve {
          defaultValue=1
          keys [
           {arriveTangent=6.4544, interpMode=RCIM_Cubic, time=0, value=0}
           {
            arriveTangent=1.5375
            arriveTangentWeight=0.1333
            interpMode=RCIM_Cubic
            leaveTangent=1.5375
            leaveTangentWeight=0.0667
            tangentMode=RCTM_User
            time=0.2845
            value=0.4847
           }
           {
            arriveTangent=-1.9657
            arriveTangentWeight=0.1296
            interpMode=RCIM_Cubic
            leaveTangent=-1.9657
            leaveTangentWeight=0.2592
            tangentMode=RCTM_User
            time=0.6793
            value=0.8725
           }
           {tangentMode=RCTM_User, time=1, value=0}
          ]
         }}}
       Scale Curve=2
      )
      B=&FloatFromCurve(
       FloatCurve=DataInterface() {propertyValues {Curve {
          defaultValue=1
          keys [
           {interpMode=RCIM_Cubic, tangentMode=RCTM_User, time=0, value=0.55}
           {interpMode=RCIM_Cubic, tangentMode=RCTM_User, time=0.5, value=1}
           {tangentMode=RCTM_User, time=1, value=0.55}
          ]
         }}}
       CurveIndex=Linked() {linkedVariable {name=NPC.FortniteNPC.TimeOfDay_0-1, type=NiagaraFloat}}
      )
     )
     B=&FloatFromCurve(
      FloatCurve=DataInterface() {propertyValues {Curve {
         defaultValue=1
         keys [
          {interpMode=RCIM_Cubic, tangentMode=RCTM_User, time=0, value=0}
          {arriveTangent=6.4829, interpMode=RCIM_Cubic, tangentMode=RCTM_Break, time=0.4, value=1}
         ]
        }}}
      CurveIndex=$Emitter.Scalability.DistanceFraction
     )
    )
    Scale RGB=&Multiply_VectorByFloat(
     Vector=<1 1 1>
     Float=&FloatFromCurve(FloatCurve=DataInterface() {propertyValues {Curve {defaultValue=1, keys [{interpMode=RCIM_Cubic, tangentMode=RCTM_User, time=0, value=6}, <at 0.7 2 RCIM_Cubic>]}}})
    )
    ScaleA=true
    ScaleRGB=true
   )
   &ScaleMeshSize(Scale Factor=&VectorFromCurve(VectorCurve=DataInterface() {propertyValues {
       XCurve {
        defaultValue=1
        keys [
         {arriveTangent=0.3265, interpMode=RCIM_Cubic, leaveTangent=0.3265, tangentMode=RCTM_User, time=0, value=0.2}
         {arriveTangent=0.4285, interpMode=RCIM_Cubic, leaveTangent=0.4285, tangentMode=RCTM_User, time=1, value=0.5}
        ]
       }
       YCurve {
        defaultValue=1
        keys [
         {arriveTangent=1.3531, interpMode=RCIM_Cubic, leaveTangent=1.3531, tangentMode=RCTM_User, time=0, value=0.25}
         {arriveTangent=0.3205, interpMode=RCIM_Cubic, leaveTangent=0.3205, tangentMode=RCTM_User, time=1, value=1}
        ]
       }
       ZCurve {
        defaultValue=1
        keys [
         {arriveTangent=1.3531, interpMode=RCIM_Cubic, leaveTangent=1.3531, tangentMode=RCTM_User, time=0, value=0.25}
         {arriveTangent=0.3205, interpMode=RCIM_Cubic, leaveTangent=0.3205, tangentMode=RCTM_User, time=1, value=1}
        ]
       }
      }})) {enabled=false, module=ScaleMeshSize001}
   &ScaleMeshSize(Scale Factor=&Multiply_Vector(
     A=&VectorFromCurve(VectorCurve=DataInterface() {propertyValues {
        XCurve {
         defaultValue=1
         keys [
          {arriveTangent=0.3265, interpMode=RCIM_Cubic, leaveTangent=0.3265, tangentMode=RCTM_User, time=0, value=0.2}
          {arriveTangent=0.4285, interpMode=RCIM_Cubic, leaveTangent=0.4285, tangentMode=RCTM_User, time=1, value=0.5}
         ]
        }
        YCurve {
         defaultValue=1
         keys [
          {arriveTangent=1.3531, interpMode=RCIM_Cubic, leaveTangent=1.3531, tangentMode=RCTM_User, time=0, value=0.25}
          {arriveTangent=0.3205, interpMode=RCIM_Cubic, leaveTangent=0.3205, tangentMode=RCTM_User, time=1, value=1}
         ]
        }
        ZCurve {
         defaultValue=1
         keys [
          {arriveTangent=1.3531, interpMode=RCIM_Cubic, leaveTangent=1.3531, tangentMode=RCTM_User, time=0, value=0.25}
          {arriveTangent=0.3205, interpMode=RCIM_Cubic, leaveTangent=0.3205, tangentMode=RCTM_User, time=1, value=1}
         ]
        }
       }})
     B=&VectorFromFloat(Value=&FloatFromCurve(
       FloatCurve=DataInterface() {propertyValues {Curve {
          defaultValue=1
          keys [
           {interpMode=RCIM_Cubic, tangentMode=RCTM_User, time=0, value=0}
           {arriveTangent=7.5246, interpMode=RCIM_Cubic, tangentMode=RCTM_Break, time=0.35, value=1}
          ]
         }}}
       CurveIndex=$Emitter.Scalability.DistanceFraction
      ))
    ))
   &CameraOffset(Camera Offset Amount=&RandomRangeFloat(Minimum=55, Maximum=222))
  ]
  rendererClasses [&NiagaraMeshRendererProperties]
  renderers [{rendererClass=&NiagaraMeshRendererProperties, rendererIndex=0}]
 }
}
summary {
 systemName=NS_Env_ZeroPoint_Orb
 emitters [
  {emitterName=Sprites, rendererClasses [&NiagaraSpriteRendererProperties]}
  {emitterName=Debris, simTarget=GPUComputeSim, rendererClasses [&NiagaraMeshRendererProperties]}
  {emitterName=Tendrills_Small, rendererClasses [&NiagaraMeshRendererProperties]}
 ]
 userVariables [User.Debris_SpawnRate {type=NiagaraFloat, defaultValue=1, description=""}]
}
